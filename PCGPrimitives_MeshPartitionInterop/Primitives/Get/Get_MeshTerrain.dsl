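; PCG primitive graph that gets mesh terrain data through Mesh Partition interop.
; It forwards target, bound size, query type and inclusivity, modifier layer priorities and channels to a core-process subgraph, whose output feeds the shared template prepare-outputs subgraph.
; Both subgraphs are exposed as graph parameters, so either stage can be overridden while keeping the same wrapper.
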(graph Get_MeshTerrain
  (params (CoreProcessGraph :type (SoftObject /Script/PCG.PCGGraphInterface) :default /PCGPrimitives_MeshPartitionInterop/Primitives/Get/Subgraphs/Get_MeshTerrain_CoreProcess.Get_MeshTerrain_CoreProcess) (PrepareOutputsGraph :type (SoftObject /Script/PCG.PCGGraphInterface) :default /PCGPrimitives/Primitives/_Template_/Primitives_PrepareOutputs.Primitives_PrepareOutputs) (MeshPartitionBoundSize :type Vector :default (1e+05 1e+05 1e+05)) (MeshPartitionTarget :type (SoftObject /Script/MeshPartition.MeshPartition)) (MeshPartitionQueryType :type (Enum /Script/PCGMeshPartitionInterop.EPCGQueryType)) (ModifierLayerPriority :type Name) (ModifierSubPriority :type Double) (MeshPartitionQueryInclusive :type Bool) (Channels :type (Array Name)))
  ;[0 0 0 0.5 32] Core Process
  (bind Subgraph_3 (Subgraph :Graph /PCGPrimitives_MeshPartitionInterop/Primitives/Get/Subgraphs/Get_MeshTerrain_CoreProcess.Get_MeshTerrain_CoreProcess :Channels Channels :MeshPartitionBoundSize MeshPartitionBoundSize :MeshPartitionQueryInclusive MeshPartitionQueryInclusive :MeshPartitionQueryType MeshPartitionQueryType :MeshPartitionTarget MeshPartitionTarget :ModifierLayerPriority ModifierLayerPriority :ModifierSubPriority ModifierSubPriority :SubgraphOverride-param CoreProcessGraph))
  ;[0 0 0 0.5 32] Prepare Output
  (bind Subgraph_4 (Subgraph :Graph /PCGPrimitives/Primitives/_Template_/Primitives_PrepareOutputs.Primitives_PrepareOutputs :SubgraphOverride-param PrepareOutputsGraph))
  (connect Subgraph_3 Out Subgraph_4 In)
  (output Out Subgraph_4))
